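# Translucent, unlit GrillMoon character overlay (Character_ScaleUp) that draws a glowing Fresnel shell, pushed out along vertex normals.
# Emissive switches between outline/interior color and a "Mystic" look, built from parallax, flowmap-distorted cracks and layered smoke, then blends in a masked underlight.
# Opacity combines Fresnel and underlight, with height fade skipped by the Weapon switch.
# MF_PostVFX_Ovrride processes the final attributes.
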Material /GrillMoon/Assets/TechArt/Character_ScaleUp/M_GrillMoon_Overlay
graph {
 properties {
  bAutomaticallySetUsageInEditor=false
  bUseMaterialAttributes=true
  bUsedWithMeshDeformer=true
  bUsedWithSkeletalMesh=true
  blendMode=BLEND_Translucent
  shadingModel=MSM_Unlit
  twoSided=true
  usedShadingModels=Unlit
 }
 nodes [
  n1 = VectorParameter() {parameterName=Color_Outline, defaultValue=<rgba 0.8829 0 1 1>}
  n2 = TextureSampleParameter2D() {
   parameterName=Normal_Map
   samplerType=SAMPLERTYPE_Normal
   texture=/BRCosmetics/Characters/Player/Male/Large/Bodies/M_LRG_UnluckyRoll/Textures/T_UnluckyRoll_Body_N
  }
  n3 = StaticSwitchParameter(True=Transform(None=@n2.RGB), False=PixelNormalWS()) {defaultValue=true, parameterName=Use normal}
  n4 = Multiply(
   A=Fresnel(ExponentIn=ScalarParameter() {defaultValue=1, parameterName=Color_Fresnel_Size}, Normal=@n3)
   B=ScalarParameter() {defaultValue=1, parameterName=Color_Fresnel_Str}
  )
  n5 = Multiply(A=@n1.RGB, B=@n4)
  n6 = NamedRerouteUsage()
  n7 = Desaturation(None=@n6)
  n8 = NamedRerouteUsage()
  n9 = ScalarParameter() {defaultValue=1, group=Grill_Moon_Myst, parameterName=Opacity_Mask_Multi}
  n10 = Saturate(None=Subtract(A=NamedRerouteUsage(), B=Multiply(A=@n8, B=@n9)))
  n11 = LinearInterpolate(
   A=@n1.RGB
   B=NamedRerouteDeclaration(None=VectorParameter() {parameterName=Color_Interior, defaultValue=<rgba 0.8829 0 1 1>}.RGB) {name=interior color, nodeColor=<rgba 0.1953 0 1 1>}
   Alpha=@n10
  )
  n12 = StaticSwitchParameter(
   True=LinearInterpolate(A=Multiply(A=@n5, B=@n7), B=NamedRerouteUsage(), Alpha=NamedRerouteUsage())
   False=Multiply(A=Multiply(A=@n11, B=@n4), B=@n7)
  ) {group=Grill_Moon_Myst, parameterName=Mystic?, sortPriority=31}
  n13 = VectorParameter() {group=Underlight, parameterName=Underlight, defaultValue=<rgba 0 10 1.9068 1>}
  n14 = Saturate(None=NamedRerouteUsage())
  n15 = Multiply(A=@n6, B=Fresnel(ExponentIn=ScalarParameter() {defaultValue=5, parameterName=Fresnel_Size}, Normal=@n3))
  n16 = Saturate(None=Add(A=NamedRerouteDeclaration(None=@n15) {name=opacity_base_mask, nodeColor=<rgba 0.8281 1 0 1>}, B=NamedRerouteUsage()))
  n17 = LinearInterpolate(
   A=ScalarParameter() {group=character_Mask, parameterName=Height_Low}
   B=ScalarParameter() {defaultValue=1.5, group=character_Mask, parameterName=Height_Str}
   Alpha=NamedRerouteUsage()
  ) {constB=2}
  n18 = Add(A=@n14, B=StaticSwitchParameter(True=@n16, False=Multiply(A=@n16, B=@n17)) {parameterName=Weapon?})
  n19 = VertexNormalWS()
  n20 = ScalarParameter() {defaultValue=1, parameterName=Shell_Distance}
  MaterialFunctionCall!(InMA=SetMaterialAttributes!(
    MaterialAttributes
    Emissive Color=LinearInterpolate(A=@n12, B=@n13.RGB, Alpha=@n14)
    Opacity=Saturate(None=@n18)
    World Position Offset=Multiply(A=@n19, B=@n20)
   ) {attributeSetTypes [#1, #2, #3]}) {materialFunction=/GrillMoon/Assets/PostProcress/MF_PostVFX_Ovrride}
  NamedRerouteDeclaration(None=TextureCoordinate()) {name=CachedUV, nodeColor=<rgba 0.3359 0 1 1>}
  n21 = NamedRerouteUsage()
  n22 = VectorParameter() {
   group=Grill_Moon_Myst
   parameterName=Myst_Crack_UV
   channelNames {a=Offset Y, b=Offset X, g=Tilling Y, r=Tilling X}
   defaultValue=<rgba 2 2 0 1>
  }
  n23 = MaterialFunctionCall(X=@n22.Tilling X, Y=@n22.Tilling X) {materialFunction=&MakeFloat2}
  n24 = MaterialFunctionCall(X=@n22.Offset X, Y=@n22.Offset X) {materialFunction=&MakeFloat2}
  n25 = VectorParameter() {group=Grill_Moon_Myst, parameterName=Myst_Flowmap_OffsetSpeed0, sortPriority=31, defaultValue=<rgba 0 0.05 0 0>}
  n26 = Panner(
   Coordinate=Add(A=Multiply(A=@n21, B=@n23.Result), B=ComponentMask(None=@n24.Result) {g=true, r=true})
   Time=Multiply(A=Time(), B=ScalarParameter() {defaultValue=0.2, group=Grill_Moon_Myst, parameterName=Crack_DistortTimeScale}) {constB=0.5}
   Speed=MaterialFunctionCall(X=@n25.R, Y=@n25.R) {materialFunction=&MakeFloat2}.Result
  )
  n27 = Multiply(A=ScalarParameter() {defaultValue=0.1, group=Grill_Moon_Myst, parameterName=FlowMap_Myst}, B=NamedRerouteUsage())
  n28 = TextureSampleParameter2D(UVs=BumpOffset(Coordinate=Add(A=@n26, B=@n27), Height=Constant() {r=-0.5})) {
   group=Grill_Moon_Myst
   parameterName=Interior_Texture
   samplerType=SAMPLERTYPE_Masks
   texture=/Game/Environments/Asteria/Terrain/Textures/T_Asteria_Ice_Cracks_M
  }
  n29 = ScalarParameter() {defaultValue=2, desc=Used to tweak str of texture, group=Grill_Moon_Myst, parameterName=Texture_adjust}
  n30 = Multiply(A=NamedRerouteUsage(), B=ScalarParameter() {defaultValue=1, group=Grill_Moon_Myst, parameterName=Crack_Mask_Multi})
  n31 = ScalarParameter() {defaultValue=2, group=Grill_Moon_Myst, parameterName=Crack_Mask_Pow}
  n32 = LinearInterpolate(
   B=StaticSwitchParameter(True=Power(Base=@n28.R, Exp=@n29), False=@n28.R) {group=Grill_Moon_Myst, parameterName=Adjust?}
   Alpha=Saturate(None=Power(Base=@n30, Exp=@n31))
  ) {constB=0}
  n33 = NamedRerouteUsage()
  NamedRerouteDeclaration(None=Multiply(
    A=StaticSwitchParameter(True=@n32, False=@n33) {group=Grill_Moon_Myst, parameterName=Mystic?, sortPriority=31}
    B=ScalarParameter() {defaultValue=1.5, parameterName=Interior_Str}
   )) {name=Interior, nodeColor=<rgba 0.6172 0 1 1>}
  n34 = Time()
  n35 = Constant4Vector() {constant=<rgba -1 1 0.5 1>}
  n36 = MaterialFunctionCall(Input=Sine(None=Multiply(A=@n34) {constB=0.5}) {period=6.283}, Input Low=@n35.R, Input High=@n35.R, Target Low=@n35.R, Target High=@n35.R) {materialFunction=&RemapValueRange}
  n37 = StaticSwitchParameter(True=LinearInterpolate(A=NamedRerouteUsage(), B=NamedRerouteUsage(), Alpha=@n36.Result), False=NamedRerouteUsage()) {group=Grill_Moon_Smoke, parameterName=EnableFeaturedSmoke}
  n38 = VectorParameter() {group=Grill_Moon_Smoke, parameterName=smokecolor_01, defaultValue=<rgba 0.3802 0.3049 0.3614 1>}
  n39 = NamedRerouteUsage()
  n40 = VectorParameter() {group=Grill_Moon_Smoke, parameterName=smokecolor_02, defaultValue=<rgba 0.2859 0.3701 0.6615 1>}
  n41 = LinearInterpolate(A=NamedRerouteUsage(), B=NamedRerouteUsage(), Alpha=@n36.Result)
  n42 = NamedRerouteUsage()
  n43 = Add(
   A=Add(A=Multiply(A=@n37, B=@n38.RGB), B=Multiply(A=@n39, B=@n40.RGB))
   B=Multiply(
    A=StaticSwitchParameter(True=@n41, False=@n42) {group=Grill_Moon_Smoke, parameterName=EnableFeaturedSmoke}
    B=VectorParameter() {group=Grill_Moon_Smoke, parameterName=smokecolor_03, defaultValue=<rgba 0.6378 0.6376 0.7604 1>}.RGB
   )
  )
  n44 = ScalarParameter() {defaultValue=1.75, group=Grill_Moon_Smoke, parameterName=SmokeStrength}
  n45 = NamedRerouteUsage()
  n46 = Multiply(
   A=Fresnel(ExponentIn=ScalarParameter() {defaultValue=2, group=Grill_Moon_Smoke, parameterName=ExponentIn}, Normal=Transform(None=@n45)) {baseReflectFraction=0}
   B=ScalarParameter() {defaultValue=0.15, group=Grill_Moon_Smoke, parameterName=FresnelStrength}
  )
  n47 = Subtract(
   A=Add(
    A=Multiply(A=@n43, B=@n44)
    B=VectorParameter() {group=Grill_Moon_Smoke, parameterName=BaseColor, defaultValue=<rgba 0.026 0.026 0.026 1>}.RGB
   )
   B=@n46
  )
  n48 = ScalarParameter() {defaultValue=1, group=Grill_Moon_Smoke, parameterName=SmokeStrengthInDark}
  NamedRerouteDeclaration(None=LinearInterpolate(A=Saturate(None=@n47), B=Multiply(A=@n43, B=@n48) {constB=10}, Alpha=@n46) {constA=1, constB=0}) {name=Smoke, nodeColor=<rgba 0 1 0.9766 1>}
  NamedRerouteDeclaration(None=@n3) {name=Normals-worldspace, nodeColor=<rgba 0.0781 0 1 1>}
  n49 = LocalPosition() {localOrigin=InstancePreSkinning}
  n50 = ShaderStageSwitch(PixelShader=VertexInterpolator(VS=@n49.XYZ).PS, VertexShader=@n49.XYZ)
  NamedRerouteDeclaration(None=SphereMask(
    A=MaterialFunctionCall(Float3=@n50) {materialFunction=&BreakOutFloat3Components}.B
    B=ScalarParameter() {defaultValue=26.0976, desc=125 was previous value, group=character_Mask, parameterName=Character_Fade_Height}
    Radius=ScalarParameter() {defaultValue=120, group=character_Mask, parameterName=Height_Fade_Radius}
    Hardness=ScalarParameter() {defaultValue=0.05, group=character_Mask, parameterName=Height_Fade_Softness}
   )) {name=height_Fade, nodeColor=<rgba 0.2656 1 0 1>}
  n51 = NamedRerouteUsage()
  n52 = VectorParameter() {
   group=Grill_Moon_Smoke
   parameterName=Smoke3_UV
   channelNames {a=Offset Y, b=Offset X, g=Tilling Y, r=Tilling X}
   defaultValue=<rgba 1.5 1.5 0 1>
  }
  n53 = MaterialFunctionCall(X=@n52.Tilling X, Y=@n52.Tilling X) {materialFunction=&MakeFloat2}
  n54 = Panner(
   Coordinate=Add(
    A=Multiply(A=@n51, B=@n53.Result)
    B=MaterialFunctionCall(X=@n52.Offset X, Y=@n52.Offset X) {materialFunction=&MakeFloat2}.Result
   )
   Speed=ComponentMask(None=VectorParameter() {group=Grill_Moon_Smoke, parameterName=Smoke3_Speed, defaultValue=<rgba 0 0.125 0 1>}.RGB) {g=true, r=true}
  )
  n55 = Multiply(A=ScalarParameter() {defaultValue=0.35, group=Grill_Moon_Smoke, parameterName=FlowMapStrength3}, B=NamedRerouteUsage())
  n56 = TextureSampleParameter2D(UVs=BumpOffset(Coordinate=Add(A=@n54, B=@n55), Height=ScalarParameter() {defaultValue=-3, group=Grill_Moon_Smoke, parameterName=Smoke3_Height})) {
   group=Grill_Moon_Smoke
   parameterName=SmokeMap
   samplerSource=SSM_Wrap_WorldGroupSettings
   samplerType=SAMPLERTYPE_Grayscale
   texture=&LowResBlurredNoise
  }
  NamedRerouteDeclaration(None=@n56.R) {name=Smoke3, nodeColor=<rgba 1 0 0.7266 1>}
  n57 = VectorParameter() {
   group=Grill_Moon_Smoke
   parameterName=Smoke_DistortUV2
   channelNames {a=Offset.Y, b=Offset.X, g=Tilling.Y, r=Tilling.X}
   defaultValue=<rgba 1 1 2 0>
  }
  n58 = Multiply(
   A=NamedRerouteUsage()
   B=MaterialFunctionCall(X=@n57.Tilling.X, Y=@n57.Tilling.X) {materialFunction=&MakeFloat2}.Result
  )
  n59 = MaterialFunctionCall(X=@n57.Offset.X, Y=@n57.Offset.X) {materialFunction=&MakeFloat2}
  n60 = VectorParameter() {group=Grill_Moon_Smoke, parameterName=Smoke_Flowmap_OffsetSpeed2, sortPriority=31, defaultValue=<rgba 0 0.05 0 0>}
  n61 = TextureSampleParameter2D(UVs=Panner(
    Coordinate=Add(A=@n58, B=@n59.Result)
    Speed=MaterialFunctionCall(X=@n60.R, Y=@n60.R) {materialFunction=&MakeFloat2}.Result
   )) {
   desc=Only RG channel works
   group=Grill_Moon_Smoke
   parameterName=Smoke_DistortUV_Flowmap
   samplerSource=SSM_Wrap_WorldGroupSettings
   samplerType=SAMPLERTYPE_LinearColor
   texture=/Game/Global/Textures/Default/Flowmaps/T_TorusFlowmap
  }
  n62 = ComponentMask(None=@n61.RGB) {g=true, r=true}
  NamedRerouteDeclaration(None=Subtract(A=Multiply(A=@n62) {constB=2})) {name=FlowMap_Secondary, nodeColor=<rgba 0 1 0.2969 1>}
  NamedRerouteDeclaration(None=MaterialFunctionCall!(XY vector=NamedRerouteUsage(), Z sign, Default Normalization vector) {materialFunction=/Engine/Functions/Engine_MaterialFunctions02/Math/DeriveNormalZ_Function}.Result) {name=Normal, nodeColor=<rgba 0.3828 1 0 1>}
  n63 = NamedRerouteUsage()
  n64 = VectorParameter() {
   group=Grill_Moon_Smoke
   parameterName=Smoke_DistortUV0
   channelNames {a=Offset.Y, b=Offset.X, g=Tilling.Y, r=Tilling.X}
   defaultValue=<rgba 1 1 2 0>
  }
  n65 = MaterialFunctionCall(X=@n64.Tilling.X, Y=@n64.Tilling.X) {materialFunction=&MakeFloat2}
  n66 = TextureSampleParameter2D(UVs=Add(
    A=Multiply(A=@n63, B=@n65.Result)
    B=MaterialFunctionCall(X=@n64.Offset.X, Y=@n64.Offset.X) {materialFunction=&MakeFloat2}.Result
   )) {
   desc=Only RG channel works
   group=Grill_Moon_Smoke
   parameterName=Smoke_DistortUV_Flowmap
   samplerSource=SSM_Wrap_WorldGroupSettings
   samplerType=SAMPLERTYPE_LinearColor
   texture=&T_TorusFlowmap
  }
  n67 = ComponentMask(None=@n66.RGB) {g=true, r=true}
  NamedRerouteDeclaration(None=Subtract(A=Multiply(A=@n67) {constB=2})) {name=Flowmap_Featured, nodeColor=<rgba 0 1 0.1797 1>}
  n68 = VectorParameter() {
   group=Grill_Moon_Smoke
   parameterName=Smoke0_UV
   channelNames {a=Offset Y, b=Offset X, g=Tilling Y, r=Tilling X}
   defaultValue=<rgba 2 2 0 1>
  }
  n69 = Add(
   A=Multiply(
    A=NamedRerouteUsage()
    B=MaterialFunctionCall(X=@n68.Tilling X, Y=@n68.Tilling X) {materialFunction=&MakeFloat2}.Result
   )
   B=ComponentMask(None=MaterialFunctionCall(X=@n68.Offset X, Y=@n68.Offset X) {materialFunction=&MakeFloat2}.Result) {g=true, r=true}
  )
  n70 = Time()
  n71 = ScalarParameter() {defaultValue=0.2, group=Grill_Moon_Smoke, parameterName=Smoke_DistortTimeScale}
  n72 = Frac(None=Multiply(A=@n70, B=@n71) {constB=0.5})
  n73 = VectorParameter() {group=Grill_Moon_Smoke, parameterName=Smoke_Flowmap_OffsetSpeed0, sortPriority=31, defaultValue=<rgba 0 0.05 0 0>}
  n74 = MaterialFunctionCall(X=@n73.R, Y=@n73.R) {materialFunction=&MakeFloat2}
  n75 = Panner(Coordinate=@n69, Time=@n72, Speed=@n74.Result)
  n76 = Multiply(A=ScalarParameter() {defaultValue=0.1, group=Grill_Moon_Smoke, parameterName=Featured_FlowMap_Strength}, B=NamedRerouteUsage())
  n77 = Constant() {r=-0.5}
  n78 = TextureSampleParameter2D(UVs=BumpOffset(Coordinate=Add(A=@n75, B=@n76), Height=@n77)) {
   group=Grill_Moon_Smoke
   parameterName=SmokeMap
   samplerSource=SSM_Wrap_WorldGroupSettings
   samplerType=SAMPLERTYPE_Grayscale
   texture=&LowResBlurredNoise
  }
  n79 = Multiply(A=@n70, B=@n71) {constB=0.5}
  n80 = Panner(Coordinate=@n69, Time=Frac(None=Add(A=@n79) {constB=0.5}), Speed=@n74.Result)
  n81 = TextureSampleParameter2D(UVs=BumpOffset(Coordinate=Add(A=@n80, B=@n76), Height=@n77)) {
   group=Grill_Moon_Smoke
   parameterName=SmokeMap
   samplerSource=SSM_Wrap_WorldGroupSettings
   samplerType=SAMPLERTYPE_Grayscale
   texture=&LowResBlurredNoise
  }
  n82 = Abs(None=Subtract(A=@n72) {constB=0.5})
  NamedRerouteDeclaration(None=LinearInterpolate(A=@n78.R, B=@n81.R, Alpha=Multiply(A=@n82) {constB=2})) {name=Smoke0, nodeColor=<rgba 0.3828 1 0 1>}
  NamedRerouteDeclaration(None=@n10) {name=interior mask, nodeColor=<rgba 1 0.5156 0 1>}
  n83 = VectorParameter() {group=Underlight, parameterName=CenterPoint, sortPriority=10, defaultValue=<rgba -25.2108 28.7725 118.6585 1>}
  n84 = SphereMask(
   A=NamedRerouteUsage()
   B=@n83.RGB
   Radius=ScalarParameter() {defaultValue=32.7146, group=Underlight, parameterName=Underlight_Radius}
   Hardness=ScalarParameter() {defaultValue=0.8, group=Underlight, parameterName=Underlight_Contrast}
  )
  n85 = DotProduct(A=Normalize(VectorInput=@n83.RGB), B=NamedRerouteUsage())
  NamedRerouteDeclaration(None=Multiply(A=Multiply(A=@n84, B=@n85), B=ScalarParameter() {defaultValue=1, group=Underlight, parameterName=Underlight_Str})) {name=underlight_Mask, nodeColor=<rgba 1 0.8906 0 1>}
  n86 = Divide(A=LocalPosition() {localOrigin=InstancePreSkinning}.XYZ, B=PreSkinnedLocalBounds().Extents)
  NamedRerouteDeclaration(None=ShaderStageSwitch(PixelShader=VertexInterpolator(VS=@n86).PS, VertexShader=@n86)) {name=Pre Skinned Local Postition, nodeColor=<rgba 1 0.5156 0 1>}
  n87 = VectorParameter() {
   group=Grill_Moon_Smoke
   parameterName=Smoke_DistortUV
   channelNames {a=Offset.Y, b=Offset.X, g=Tilling.Y, r=Tilling.X}
   defaultValue=<rgba 1 1 0 0>
  }
  n88 = Add(
   A=Multiply(
    A=NamedRerouteUsage()
    B=MaterialFunctionCall(X=@n87.Tilling.X, Y=@n87.Tilling.X) {materialFunction=&MakeFloat2}.Result
   )
   B=MaterialFunctionCall(X=@n87.Offset.X, Y=@n87.Offset.X) {materialFunction=&MakeFloat2}.Result
  )
  n89 = VectorParameter() {group=Grill_Moon_Smoke, parameterName=Smoke_Flowmap_OffsetSpeed, sortPriority=31}
  n90 = MaterialFunctionCall(X=@n89.R, Y=@n89.R) {materialFunction=&MakeFloat2}
  n91 = NamedRerouteUsage()
  n92 = ScalarParameter() {group=Grill_Moon_Smoke, parameterName=FlowMap_Secondary_Scale}
  n93 = TextureSampleParameter2D(UVs=Add(A=Panner(Coordinate=@n88, Speed=@n90.Result), B=Multiply(A=@n91, B=@n92))) {
   desc=Only RG channel works
   group=Grill_Moon_Smoke
   parameterName=Smoke_DistortUV_Flowmap
   samplerSource=SSM_Wrap_WorldGroupSettings
   samplerType=SAMPLERTYPE_LinearColor
   texture=&T_TorusFlowmap
  }
  n94 = ComponentMask(None=@n93.RGB) {g=true, r=true}
  NamedRerouteDeclaration(None=Subtract(A=Multiply(A=@n94) {constB=2})) {name=FlowMap_Main, nodeColor=<rgba 0 1 0.1797 1>}
  n95 = NamedRerouteUsage()
  n96 = VectorParameter() {
   group=Grill_Moon_Smoke
   parameterName=Smoke1_UV
   channelNames {a=Offset Y, b=Offset X, g=Tilling Y, r=Tilling X}
   defaultValue=<rgba 2 2 0 1>
  }
  n97 = MaterialFunctionCall(X=@n96.Tilling X, Y=@n96.Tilling X) {materialFunction=&MakeFloat2}
  n98 = MaterialFunctionCall(X=@n96.Offset X, Y=@n96.Offset X) {materialFunction=&MakeFloat2}
  n99 = Panner(
   Coordinate=Add(A=Multiply(A=@n95, B=@n97.Result), B=ComponentMask(None=@n98.Result) {g=true, r=true})
   Speed=ComponentMask(None=VectorParameter() {group=Grill_Moon_Smoke, parameterName=Smoke1_Speed, defaultValue=<rgba 0 0.1 0 0>}.RGB) {g=true, r=true}
  )
  n100 = Multiply(A=ScalarParameter() {defaultValue=0.1, group=Grill_Moon_Smoke, parameterName=FlowMapStrength1}, B=NamedRerouteUsage())
  n101 = TextureSampleParameter2D(UVs=BumpOffset(Coordinate=Add(A=@n99, B=@n100), Height=ScalarParameter() {group=Grill_Moon_Smoke, parameterName=Smoke1_Height})) {
   group=Grill_Moon_Smoke
   parameterName=SmokeMap
   samplerSource=SSM_Wrap_WorldGroupSettings
   samplerType=SAMPLERTYPE_Grayscale
   texture=&LowResBlurredNoise
  }
  NamedRerouteDeclaration(None=@n101.RGB) {name=Smoke1, nodeColor=<rgba 0.3828 1 0 1>}
  n102 = NamedRerouteUsage()
  n103 = VectorParameter() {
   group=Grill_Moon_Smoke
   parameterName=Smoke2_UV
   channelNames {a=Offset Y, b=Offset X, g=Tilling Y, r=Tilling X}
   defaultValue=<rgba 1.25 1.25 0 1>
  }
  n104 = MaterialFunctionCall(X=@n103.Tilling X, Y=@n103.Tilling X) {materialFunction=&MakeFloat2}
  n105 = Panner(
   Coordinate=Add(
    A=Multiply(A=@n102, B=@n104.Result)
    B=MaterialFunctionCall(X=@n103.Offset X, Y=@n103.Offset X) {materialFunction=&MakeFloat2}.Result
   )
   Speed=ComponentMask(None=VectorParameter() {group=Grill_Moon_Smoke, parameterName=Smoke2_Speed, defaultValue=<rgba 0 0.15 0 1>}.RGB) {g=true, r=true}
  )
  n106 = Multiply(A=ScalarParameter() {defaultValue=0.33, group=Grill_Moon_Smoke, parameterName=FlowMapStrength}, B=NamedRerouteUsage())
  n107 = TextureSampleParameter2D(UVs=BumpOffset(Coordinate=Add(A=@n105, B=@n106), Height=ScalarParameter() {defaultValue=-1.5, group=Grill_Moon_Smoke, parameterName=Smoke2_Height})) {
   group=Grill_Moon_Smoke
   parameterName=SmokeMap
   samplerSource=SSM_Wrap_WorldGroupSettings
   samplerType=SAMPLERTYPE_Grayscale
   texture=&LowResBlurredNoise
  }
  NamedRerouteDeclaration(None=@n107.R) {name=Smoke2, nodeColor=<rgba 1 0 0.7031 1>}
 ]
}
